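# Editor-utility component that spawns and destroys heatmap element actors for its owning heatmap renderer manager.
# spawnElement reads a location from the data object at an index, resolves the soft element class and spawns it through EditorActorSubsystem.
# The new element receives the data object, index, owner and owning manager before it is returned.
# destroyElements passes an element array to the same subsystem's DestroyActors.
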EditorUtilityBlueprint /Game/Blueprints/HeatmapRenderer/Editor/HeatmapRenderer_EditorComponent
graphs [
 EventGraph {nodes [Event Begin Play()]}
 destroyElements {nodes [
   n1 = destroyElements!() -> (then, elements:Array of BP Heatmap Renderer Element Object References) {is=title K2Node_FunctionEntry}
   > DestroyActors!(execute, self:Editor Actor Subsystem Object Reference=Get EditorActorSubsystem(), ActorsToDestroy:Array of Actor Object References=@n1) -> (then, ReturnValue:bool)
   Return Node!(execute, " ":bool=false)
  ]}
 spawnElement {nodes [
   n1 = spawnElement!() -> (then, class:BP Heatmap Renderer Element Soft Class Reference, data:"Heatmap Renderer Data Object Object Reference (by ref)", dataIndex:int=0) {is=title K2Node_FunctionEntry}
   > n2 = getBasicDataOutput!(execute, self:Heatmap Renderer Data Object Object Reference=@n1.data, index:int=@n1.dataIndex) -> (then, Location:Vector, Severity:double)
   > n3 = SpawnActorFromClass!(
    execute
    self:Editor Actor Subsystem Object Reference=Get EditorActorSubsystem()
    ActorClass:Actor Class Reference=Resolve Soft Reference!(Input:BP Heatmap Renderer Element Soft Class Reference=@n1) -> (Output:BP Heatmap Renderer Element Class Reference)
    Location:Vector=@n2
    Rotation:Rotator
    bTransient:bool
   ) -> (then, ReturnValue:BP Heatmap Renderer Element Object Reference)
   > Set CurrentDataObject!(execute, CurrentDataObject:Heatmap Renderer Data Object Object Reference=@n1.data, self:BP Heatmap Renderer Element Object Reference=@n3) -> (then, Output_Get:Heatmap Renderer Data Object Object Reference)
   > Set DataIndex!(execute, DataIndex:int=@n1.dataIndex, self:BP Heatmap Renderer Element Object Reference=@n3) -> (then, Output_Get:int=0)
   n4 = GetOwner()
   > SetOwner!(execute, self:Actor Object Reference=@n3, NewOwner:Actor Object Reference=@n4) -> (then)
   > Set OwningManager!(
    execute
    OwningManager:Heatmap Renderer Manager Object Reference=Cast To BP_HeatmapRenderer_Manager!(Object=@n4) -> (AsHeatmap Renderer Manager:Heatmap Renderer Manager Object Reference, bSuccess:bool=~) {is=title K2Node_DynamicCast}
    self:BP Heatmap Renderer Element Object Reference=@n3
   ) -> (then, Output_Get:Heatmap Renderer Manager Object Reference)
   > Return Node!(execute, newElement:BP Heatmap Renderer Element Object Reference=@n3)
  ]}
]
variables []
