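# Single Layer Water material for the heatmap renderer's height map, by its name, built around mesh vertex colors.
# One branch raises vertex alpha to a power and remaps it between min and max parameters as a blend factor; another scales a self-weighted vertex color.
# Neither branch shows a connection to a material output.
# The water output receives only constant scattering and a parameter scaling color behind water.
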Material /Game/Blueprints/HeatmapRenderer/Content/Materials/M_HeatmapRenderer_HeightMap
graph {
 properties {pixelDepthOffsetMode=PDOM_Legacy, shadingModel=MSM_SingleLayerWater, usedShadingModels=SingleLayerWater}
 nodes [
  Constant()
  n1 = VertexColor()
  n2 = ScalarParameter() {defaultValue=0.6, parameterName=PowAlpha}
  LinearInterpolate(Alpha=LinearInterpolate(A=ScalarParameter() {parameterName=MinAlpha}, B=ScalarParameter() {defaultValue=2, parameterName=MaxAlpha}, Alpha=Power(Base=@n1.A, Exp=@n2)))
  Multiply(A=ScalarParameter() {defaultValue=2, parameterName=dsaa}, B=LinearInterpolate(B=@n1, Alpha=@n1))
  SingleLayerWaterMaterialOutput!(
   ScatteringCoefficients=Constant()
   AbsorptionCoefficients
   PhaseG
   ColorScaleBehindWater=ScalarParameter() {defaultValue=0.25, parameterName=ScaleBehindWater}
  )
 ]
}
