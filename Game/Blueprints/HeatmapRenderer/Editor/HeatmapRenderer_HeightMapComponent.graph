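# Editor heatmap renderer component that draws the current data object of its owning heatmap manager as a colored height-map mesh.
# On drawMesh it finds or adds a DynamicMeshComponent, then Delaunay-triangulates the data locations' XY points using Geometry Script.
# Normalized severities sample the manager's color curve for vertex colors and, remapped, set vertex heights before PN tessellation and iterative smoothing.
# The spawnElement and destroyElements functions are empty stubs.
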EditorUtilityBlueprint /Game/Blueprints/HeatmapRenderer/Editor/HeatmapRenderer_HeightMapComponent
graphs [
 DrawMeshInternal {nodes [
   DrawMeshInternal() {is=title K2Node_FunctionEntry}
   > n1 = Sequence!(execute) -> (then_0, then_1, then_2:Exec, then_3:Exec, then_4:Exec)
   > n2 = GetDynamicMeshComp!(execute, self) -> (then, Component:Dynamic Mesh Component Object Reference)
   > Set l_meshComp(l_meshComp:Dynamic Mesh Component Object Reference=@n2) -> (Output_Get:Dynamic Mesh Component Object Reference)
   n3 = Cast To BP_HeatmapRenderer_Manager!(execute=@n1.then_1, Object=GetOwner()) -> (then, CastFailed, AsHeatmap Renderer Manager:Heatmap Renderer Manager Object Reference) {is=title K2Node_DynamicCast}
   n4 = Get CurrentDataObject!(self:Heatmap Renderer Manager Object Reference=@n3) -> (CurrentDataObject:Heatmap Renderer Data Object Object Reference)
   > n5 = GetSeveritiesNormalized!(execute, self:Heatmap Renderer Data Object Object Reference=@n4) -> (then, Normalized:"Array of Float (double-precision)s")
   > Set l_severities(l_severities:"Array of Float (double-precision)s"=@n5) -> (Output_Get:"Array of Float (double-precision)s")
   > Set l_curve(l_curve:Curve Linear Color Object Reference=Get ColorCurve!(self:Heatmap Renderer Manager Object Reference=@n3) -> (ColorCurve:Curve Linear Color Object Reference)) -> (Output_Get:Curve Linear Color Object Reference)
   > n6 = For Each Loop!(Exec, Array:Array of Vectors=Get Locations!(self:Heatmap Renderer Data Object Object Reference=@n4) -> (Locations:Array of Vectors)) -> (LoopBody, Array Element_X:double, Array Element_Y:double, Array Element_Z:double, Array Index, Completed)
   n7 = For Each Loop!(Exec=@n6.Completed, Array:"Array of Float (double-precision)s"=Get l_severities() -> (l_severities:"Array of Float (double-precision)s")) -> (LoopBody, Array Element:double, Array Index, Completed)
   > Add(
    TargetArray:"Array of Float (double-precision)s"=Get l_severitiesAdjusted() -> (l_severitiesAdjusted:"Array of Float (double-precision)s")
    NewItem:"Float (double-precision) (by ref)"=NormalizeToRange!(Value:double=@n7, RangeMin:double=0.2, RangeMax:double=1) -> (ReturnValue:double)
   ) {is=title K2Node_CallArrayFunction}
   Add!(
    execute=@n6.LoopBody
    TargetArray:Array of Vector 2D Structures=Get l_verts() -> (l_verts:Array of Vector 2D Structures)
    NewItem_X:double=@n6
    NewItem_Y:double=@n6.Array Element_Y
   ) -> (then, ReturnValue)
   n8 = Get l_meshComp() -> (l_meshComp:Dynamic Mesh Component Object Reference)
   SetMaterial(execute=@n1.then_2, self=@n8, ElementIndex=0, Material=/Game/Blueprints/HeatmapRenderer/Content/Materials/M_HeatmapRenderer_HeightMap)
   > n9 = GetDynamicMesh!(execute, self:Base Dynamic Mesh Component Object Reference=@n8) -> (then, ReturnValue:Dynamic Mesh Object Reference)
   > n10 = Set l_mesh(l_mesh:Dynamic Mesh Object Reference=@n9) -> (Output_Get:Dynamic Mesh Object Reference)
   > n11 = Reset!(execute, self:Dynamic Mesh Object Reference=@n10) -> (then, ReturnValue:Dynamic Mesh Object Reference)
   > n12 = AppendDelaunayTriangulation2D!(
    execute
    TargetMesh:Dynamic Mesh Object Reference=@n11
    PrimitiveOptions:Geometry Script Primitive Options Structure=Make Geometry Script Primitive Options!(
     PolygroupMode:EGeometryScriptPrimitivePolygroupMode Enum=PerFace
     bFlipOrientation:bool=False
     UVMode:EGeometryScriptPrimitiveUVMode Enum=Uniform
     MaterialID:int=0
    ) -> (GeometryScriptPrimitiveOptions:Geometry Script Primitive Options Structure) {is=title K2Node_MakeStruct}
    Transform:Transform
    VertexPositions:Array of Vector 2D Structures=Get l_verts()
    ConstrainedEdges:Array of Int Point Structures
    TriangulationOptions:Geometry Script Constrained Delaunay Triangulation Options Structure=Make Geometry Script Constrained Delaunay Triangulation Options!(ConstrainedEdgesFillMode:EGeometryScriptPolygonFillMode Enum=All, bValidateEdgesInResult:bool=True, bRemoveDuplicateVertices:bool=False) -> (GeometryScriptConstrainedDelaunayTriangulationOptions:Geometry Script Constrained Delaunay Triangulation Options Structure) {is=title K2Node_MakeStruct}
   ) -> (then, PositionsToVertexIDs:Array of Integers, bHasDuplicateVertices:bool, ReturnValue:Dynamic Mesh Object Reference)
   > Set l_vertexIDs(l_vertexIDs:Array of Integers=@n12) -> (Output_Get:Array of Integers)
   n13 = SetMeshConstantVertexColor!(
    execute=@n1.then_3
    TargetMesh:Dynamic Mesh Object Reference=Get l_mesh() -> (l_mesh:Dynamic Mesh Object Reference)
    Color:Linear Color Structure=GetClampedLinearColorValue!(self:Curve Linear Color Object Reference=Get l_curve() -> (l_curve:Curve Linear Color Object Reference), InTime:float=0) -> (ReturnValue:Linear Color Structure)
    Flags:Geometry Script Color Flags Structure
    bClearExisting:bool=true
   ) -> (then, ReturnValue:Dynamic Mesh Object Reference)
   > n14 = GetMeshPerVertexColors!(execute, TargetMesh:Dynamic Mesh Object Reference=@n13, bBlendSplitVertexValues:bool=true) -> (then, ColorList:Color List Structure, bIsValidColorSet:bool, bHasVertexIDGaps:bool, ReturnValue:Dynamic Mesh Object Reference)
   > Set l_colorList(l_colorList:Color List Structure=@n14) -> (Output_Get:Color List Structure)
   > n15 = For Each Loop(Array:Array of Integers=Get l_vertexIDs() -> (l_vertexIDs:Array of Integers)) -> (Array Element:int)
   SetMeshPerVertexColors!(
    execute=@n15.Completed
    TargetMesh:Dynamic Mesh Object Reference=@n14.ReturnValue
    VertexColorList:Color List Structure=Get l_colorList() -> (l_colorList:Color List Structure)
   ) -> (then, ReturnValue:Dynamic Mesh Object Reference)
   n16 = Get l_severities()
   SetColorListItem!(
    execute=@n15.LoopBody
    ColorList:"Color List Structure (by ref)"=Get l_colorList()
    Index:int=@n15
    NewColor:Linear Color Structure=GetClampedLinearColorValue(
     self=Get l_curve()
     InTime="Get (a ref)"(Array:"Array of Float (double-precision)s"=@n16, Dimension 1=@n15.Array Index) -> (Output:"Float (double-precision) (by ref)")
    )
   ) -> (then, bIsValidIndex:bool)
   n17 = For Each Loop(Exec=@n1.then_4, Array=Get l_vertexIDs())
   n18 = Apply PN Tessellation!(
    execute=@n17.Completed
    TargetMesh:Dynamic Mesh Object Reference=Get l_mesh()
    Options:PN Tessellate Options Structure=Make PN Tessellate Options!(bRecomputeNormals:bool=True) -> (GeometryScriptPNTessellateOptions:PN Tessellate Options Structure) {is=title K2Node_MakeStruct}
    TessellationLevel:int=5
   ) -> (then, ReturnValue:Dynamic Mesh Object Reference)
   > ApplyIterativeSmoothingToMesh!(
    execute
    TargetMesh:Dynamic Mesh Object Reference=@n18
    Selection:Mesh Selection Structure
    Options:Geometry Script Iterative Mesh Smoothing Options Structure=Make Geometry Script Iterative Mesh Smoothing Options!(NumIterations:int=10, Alpha:float=0.2, EmptyBehavior:EGeometryScriptEmptySelectionBehavior Enum=FullMeshSelection) -> (GeometryScriptIterativeMeshSmoothingOptions:Geometry Script Iterative Mesh Smoothing Options Structure) {is=title K2Node_MakeStruct}
   ) -> (then, ReturnValue:Dynamic Mesh Object Reference)
   n19 = Get l_mesh()
   n20 = GetVertexPosition!(TargetMesh:Dynamic Mesh Object Reference=@n19, VertexID:int=@n17) -> (bIsValidVertex:bool, ReturnValue_X:double, ReturnValue_Y:double, ReturnValue_Z:double)
   n21 = Get l_severitiesAdjusted()
   SetVertexPosition!(
    execute=@n17.LoopBody
    TargetMesh:Dynamic Mesh Object Reference=@n19
    VertexID:int=@n17
    NewPosition_X:double=@n20.ReturnValue_X
    NewPosition_Y:double=@n20.ReturnValue_Y
    NewPosition_Z:double=float * float(A="Get (a ref)"(Array=@n21, Dimension 1=@n17.Array Index), B=50000) {is=title K2Node_PromotableOperator}
    bDeferChangeNotifications:bool=true
   ) -> (then, bIsValidVertex:bool, ReturnValue:Dynamic Mesh Object Reference)
  ]}
 EventGraph {nodes [
   Event drawMesh() {is=title K2Node_Event}
   > DrawMeshInternal() {is=title K2Node_CallFunction}
   Event Begin Play()
  ]}
 GetDynamicMeshComp {nodes [
   GetDynamicMeshComp() {is=title K2Node_FunctionEntry}
   > n1 = Set l_meshComp(l_meshComp=GetComponentByClass!(self:Actor Object Reference=GetOwner(), ComponentClass:Actor Component Class Reference=/Script/GeometryFramework.DynamicMeshComponent) -> (ReturnValue:Dynamic Mesh Component Object Reference))
   > n2 = Is Valid(InputObject=@n1)
   n3 = Add Component by Class!(
    execute:Exec=@n2.Is Not Valid
    self:Actor Object Reference=GetOwner()
    WorldContextObject:Object Reference
    Class:Actor Component Class Reference=&DynamicMeshComponent
    bManualAttachment:bool
    RelativeTransform:Transform
   ) -> (then:Exec, ReturnValue:Dynamic Mesh Component Object Reference) {is=title K2Node_AddComponentByClass}
   > n4 = Set l_meshComp(l_meshComp=@n3)
   > Return Node!(execute, Component:Dynamic Mesh Component Object Reference=@n4)
   Return Node(execute=@n2.Is Valid, Component=Get l_meshComp())
  ]}
 destroyElements {nodes [
   destroyElements!() -> (then, elements:Array of BP Heatmap Renderer Element Object References) {is=title K2Node_FunctionEntry}
   > Return Node!(execute, " ":bool=false)
  ]}
 spawnElement {nodes [
   spawnElement!() -> (then, class:BP Heatmap Renderer Element Soft Class Reference, data:"Heatmap Renderer Data Object Object Reference (by ref)", dataIndex:int=0) {is=title K2Node_FunctionEntry}
   > Return Node!(execute, newElement:BP Heatmap Renderer Element Object Reference)
  ]}
]
variables []
