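; I_HeatmapRenderer_EditorBridge is a Blueprint interface in the heatmap renderer's Editor folder, declaring the operations of an editor bridge.
; spawnElement takes a soft element class, a data object by reference and a data index, while destroyElements takes an array of elements and returns a success-style boolean.
; drawMesh takes no parameters, and all three functions are declarations only, leaving the behavior to implementers.
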(fn destroyElements
  (params (elements :type Array%20of%20BP%20Heatmap%20Renderer%20Element%20Object%20References))
  (return false))

(fn drawMesh)

(fn spawnElement
  (params (class :type BP%20Heatmap%20Renderer%20Element%20Soft%20Class%20Reference) (data :type Heatmap%20Renderer%20Data%20Object%20Object%20Reference%20%28by%20ref%29) (dataIndex :type Integer))
  (return 0))
